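# Physics asset for the medium male Grasshopper player cosmetic body, covering a hip sash and the pelvis and thighs around it.
# Capsules on the pelvis and both thighs form a coarse lower-body shape, with free-rotation constraints linking the thighs to the pelvis.
# The sash root also attaches freely to the pelvis and branches into left and right three-segment capsule chains with tight twist limits and wider swing limits.
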PhysicsAsset /BRCosmetics/Characters/Player/Male/Medium/Bodies/M_MED_Grasshopper/Meshes/M_MED_Grasshopper_Physics
bodies {
 dyn_hipSash [{shapeType=Capsule, length=1, radius=1, shapeName=dyn_hipSash_capsule}]
 dyn_hipSash_01_l [{shapeType=Capsule, length=10, radius=2, shapeName=dyn_hipSash_01_l_capsule, center=<2.7943 0 0.0002>, rotation=<rot 0 0 90>}]
 dyn_hipSash_01_r [{shapeType=Capsule, length=10, radius=2, shapeName=dyn_hipSash_01_r_capsule, center=<2.7943 0 0.0002>, rotation=<rot 0 0 90>}]
 dyn_hipSash_02_l [{shapeType=Capsule, length=10, radius=2, shapeName=dyn_hipSash_02_l_capsule, center=<2.7943 0 0.0002>, rotation=<rot 0 0 90>}]
 dyn_hipSash_02_r [{shapeType=Capsule, length=10, radius=2, shapeName=dyn_hipSash_02_r_capsule, center=<2.7943 0 0.0002>, rotation=<rot 0 0 90>}]
 dyn_hipSash_03_l [{shapeType=Capsule, length=10, radius=2, shapeName=dyn_hipSash_03_l_capsule, center=<2.7943 0 0.0002>, rotation=<rot 0 0 90>}]
 dyn_hipSash_03_r [{shapeType=Capsule, length=10, radius=2, shapeName=dyn_hipSash_03_r_capsule, center=<2.7943 0 0.0002>, rotation=<rot 0 0 90>}]
 pelvis [{
   shapeType=Capsule
   length=15.9844
   radius=10.1635
   shapeName=pelvis_capsule
   center=<0.1953 -0.9993 -0.0895>
   rotation=<rot -8.9821 -90 0.0264>
  }]
 thigh_l [{shapeType=Capsule, length=24.0156, radius=8.9068, center=<-20.3342 -0.587 0.7879>, rotation=<rot 0 -88.7712 -84.2023>}]
 thigh_r [{shapeType=Capsule, length=24.0335, radius=8.964, center=<20.3714 0.5939 -0.75>, rotation=<rot 0 -89.2353 -83.7293>}]
}
constraints [
 {
  bone1Name=thigh_l
  bone2Name=pelvis
  swing1LimitDegrees=45
  swing1Motion=Free
  swing2LimitDegrees=45
  swing2Motion=Free
  twistLimitDegrees=45
  twistMotion=Free
 }
 {
  bone1Name=thigh_r
  bone2Name=pelvis
  swing1LimitDegrees=45
  swing1Motion=Free
  swing2LimitDegrees=45
  swing2Motion=Free
  twistLimitDegrees=45
  twistMotion=Free
 }
 {bone1Name=dyn_hipSash_01_l, bone2Name=dyn_hipSash, swing1LimitDegrees=45, swing2LimitDegrees=45, twistLimitDegrees=5}
 {bone1Name=dyn_hipSash_02_l, bone2Name=dyn_hipSash_01_l, swing1LimitDegrees=45, swing2LimitDegrees=45, twistLimitDegrees=5}
 {bone1Name=dyn_hipSash_03_l, bone2Name=dyn_hipSash_02_l, swing1LimitDegrees=45, swing2LimitDegrees=45, twistLimitDegrees=5}
 {bone1Name=dyn_hipSash_01_r, bone2Name=dyn_hipSash, swing1LimitDegrees=45, swing2LimitDegrees=45, twistLimitDegrees=5}
 {bone1Name=dyn_hipSash_02_r, bone2Name=dyn_hipSash_01_r, swing1LimitDegrees=45, swing2LimitDegrees=45, twistLimitDegrees=5}
 {bone1Name=dyn_hipSash_03_r, bone2Name=dyn_hipSash_02_r, swing1LimitDegrees=45, swing2LimitDegrees=45, twistLimitDegrees=5}
 {
  bone1Name=dyn_hipSash
  bone2Name=pelvis
  swing1LimitDegrees=45
  swing1Motion=Free
  swing2LimitDegrees=45
  swing2Motion=Free
  twistLimitDegrees=45
  twistMotion=Free
 }
]
