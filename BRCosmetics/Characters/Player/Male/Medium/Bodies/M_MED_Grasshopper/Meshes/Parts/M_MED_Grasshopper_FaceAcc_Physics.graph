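# Physics asset for the face accessory mesh of the Grasshopper male medium player cosmetic, setting up simulated motion for its dyn_hair bones.
# Small capsules on seven dyn_hair bones form short chains from the head, one branching into three, with tight swing and twist limits.
# Larger capsules on head, spine_05 and spine_04 link through free-rotation constraints, giving the body volumes alongside the hair.
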PhysicsAsset /BRCosmetics/Characters/Player/Male/Medium/Bodies/M_MED_Grasshopper/Meshes/Parts/M_MED_Grasshopper_FaceAcc_Physics
bodies {
 dyn_hair_0 [{shapeType=Capsule, length=1.5, radius=0.75, shapeName=dyn_hair_0_capsule, center=<1.4861 0 0>, rotation=<rot 90 0 0>}]
 dyn_hair_1 [{shapeType=Capsule, length=1.5, radius=0.75, shapeName=dyn_hair_1_capsule, center=<1.4861 0 0>, rotation=<rot 90 0 0>}]
 dyn_hair_2 [{shapeType=Capsule, length=1.5, radius=0.75, shapeName=dyn_hair_2_capsule, center=<1.4861 0 0>, rotation=<rot 90 0 0>}]
 dyn_hair_3 [{shapeType=Capsule, length=1.5, radius=0.75, shapeName=dyn_hair_3_capsule, center=<1.4861 0 0>, rotation=<rot 90 0 0>}]
 dyn_hair_4 [{shapeType=Capsule, length=1.5, radius=0.75, shapeName=dyn_hair_4_capsule, center=<1.4861 0 0>, rotation=<rot 90 0 0>}]
 dyn_hair_5 [{shapeType=Capsule, length=0.5, radius=1.75, shapeName=dyn_hair_5_capsule, center=<1.4861 0 0>, rotation=<rot 90 0 0>}]
 dyn_hair_6 [{shapeType=Capsule, length=1.5, radius=0.75, shapeName=dyn_hair_6_capsule, center=<1.4861 0 0>, rotation=<rot 90 0 0>}]
 head [{shapeType=Capsule, length=11.329, radius=6.6914, shapeName=head_capsule, center=<1.2343 2.0526 0.054>, rotation=<rot 0 -74.9754 -87.4377>}]
 spine_04 [{
   shapeType=Capsule
   length=12.4722
   radius=7.6699
   shapeName=spine_05_capsule
   center=<-0.3492 -0.1887 -0.1289>
   rotation=<rot 1.8755 -90 -0.7266>
  }]
 spine_05 [{
   shapeType=Capsule
   length=12.4722
   radius=8.4199
   shapeName=spine_05_capsule
   center=<-0.3492 -0.1887 -0.1289>
   rotation=<rot 1.8755 -90 -0.7266>
  }]
}
constraints [
 {
  bone1Name=spine_05
  bone2Name=spine_04
  swing1LimitDegrees=45
  swing1Motion=Free
  swing2LimitDegrees=45
  swing2Motion=Free
  twistLimitDegrees=45
  twistMotion=Free
 }
 {bone1Name=dyn_hair_0, bone2Name=head, swing1LimitDegrees=10, swing2LimitDegrees=20, twistLimitDegrees=5}
 {bone1Name=dyn_hair_1, bone2Name=dyn_hair_0, swing1LimitDegrees=10, swing2LimitDegrees=20, twistLimitDegrees=5}
 {bone1Name=dyn_hair_2, bone2Name=dyn_hair_0, swing1LimitDegrees=10, swing2LimitDegrees=20, twistLimitDegrees=5}
 {bone1Name=dyn_hair_3, bone2Name=dyn_hair_0, swing1LimitDegrees=10, swing2LimitDegrees=20, twistLimitDegrees=5}
 {bone1Name=dyn_hair_4, bone2Name=head, swing1LimitDegrees=10, swing2LimitDegrees=20, twistLimitDegrees=5}
 {bone1Name=dyn_hair_5, bone2Name=dyn_hair_4, swing1LimitDegrees=10, swing2LimitDegrees=20, twistLimitDegrees=5}
 {bone1Name=dyn_hair_6, bone2Name=head, swing1LimitDegrees=10, swing2LimitDegrees=20, twistLimitDegrees=5}
 {
  bone1Name=head
  bone2Name=spine_05
  swing1LimitDegrees=45
  swing1Motion=Free
  swing2LimitDegrees=45
  swing2Motion=Free
  twistLimitDegrees=45
  twistMotion=Free
 }
]
