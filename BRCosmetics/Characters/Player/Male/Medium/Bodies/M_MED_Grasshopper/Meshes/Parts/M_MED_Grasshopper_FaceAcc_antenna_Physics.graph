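# Physics asset for the antenna face accessory of the medium male Grasshopper character body.
# Each side is a chain of three capsule bodies on the dyn_antenna bones, anchored to a small capsule on its MSTR root bone.
# Constraints link each segment to the one before it with identical, tight swing and twist limits, so each antenna can bend only slightly at every joint.
# The left and right chains share the same shapes and limits.
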PhysicsAsset /BRCosmetics/Characters/Player/Male/Medium/Bodies/M_MED_Grasshopper/Meshes/Parts/M_MED_Grasshopper_FaceAcc_antenna_Physics
bodies {
 dyn_antenna_l_01 [{shapeType=Capsule, length=3.2, radius=3.2, shapeName=dyn_antenna_r_01_capsule, center=<5.2 0 0>, rotation=<rot 90 0 0>}]
 dyn_antenna_l_01_MSTR [{shapeType=Capsule, length=0.505, radius=0.505, shapeName=dyn_antenna_l_01_MSTR_capsule}]
 dyn_antenna_l_02 [{shapeType=Capsule, length=3.2, radius=3.2, shapeName=dyn_antenna_r_01_capsule, center=<5.2 0 0>, rotation=<rot 90 0 0>}]
 dyn_antenna_l_03 [{shapeType=Capsule, length=3.2, radius=3.2, shapeName=dyn_antenna_r_01_capsule, center=<5.2 0 0>, rotation=<rot 90 0 0>}]
 dyn_antenna_r_01 [{shapeType=Capsule, length=3.2, radius=3.2, shapeName=dyn_antenna_r_01_capsule, center=<5.2 0 0>, rotation=<rot 90 0 0>}]
 dyn_antenna_r_01_MSTR [{shapeType=Capsule, length=0.505, radius=0.505, shapeName=dyn_antenna_r_01_MSTR_capsule}]
 dyn_antenna_r_02 [{shapeType=Capsule, length=3.2, radius=3.2, shapeName=dyn_antenna_r_01_capsule, center=<5.2 0 0>, rotation=<rot 90 0 0>}]
 dyn_antenna_r_03 [{shapeType=Capsule, length=3.2, radius=3.2, shapeName=dyn_antenna_r_01_capsule, center=<5.2 0 0>, rotation=<rot 90 0 0>}]
}
constraints [
 {bone1Name=dyn_antenna_r_01, bone2Name=dyn_antenna_r_01_MSTR, swing1LimitDegrees=20, swing2LimitDegrees=20, twistLimitDegrees=20}
 {bone1Name=dyn_antenna_r_02, bone2Name=dyn_antenna_r_01, swing1LimitDegrees=20, swing2LimitDegrees=20, twistLimitDegrees=20}
 {bone1Name=dyn_antenna_r_03, bone2Name=dyn_antenna_r_02, swing1LimitDegrees=20, swing2LimitDegrees=20, twistLimitDegrees=20}
 {bone1Name=dyn_antenna_l_01, bone2Name=dyn_antenna_l_01_MSTR, swing1LimitDegrees=20, swing2LimitDegrees=20, twistLimitDegrees=20}
 {bone1Name=dyn_antenna_l_02, bone2Name=dyn_antenna_l_01, swing1LimitDegrees=20, swing2LimitDegrees=20, twistLimitDegrees=20}
 {bone1Name=dyn_antenna_l_03, bone2Name=dyn_antenna_l_02, swing1LimitDegrees=20, swing2LimitDegrees=20, twistLimitDegrees=20}
]
